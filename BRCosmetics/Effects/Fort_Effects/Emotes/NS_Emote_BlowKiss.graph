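# Niagara system for the blow-kiss emote that launches a big heart and a burst of little hearts as mesh particles.
# The hearts drift upward under drag, scale along curves and generate death events, with the big one also spinning down.
# Pop and PopSmall have no spawn module, presumably spawning from those events as brief shrinking bursts.
# All four share a system-level HeartEmissive color and use distance culling; the Empty emitter is disabled.
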NiagaraSystem /BRCosmetics/Effects/Fort_Effects/Emotes/NS_Emote_BlowKiss
dependencies {
 usedDataInterfaces [&NiagaraDataInterfaceCurve]
 usedDynamicInputs [&Multiply_VectorByFloat]
 usedModules [
  &self:SystemSpawnScript.SystemScriptSource.SystemScriptGraph.NiagaraNodeAssignment_0.SetVariables_#1
  &SystemState
 ]
}
emitters {
 BigHeart {
  emitterUpdateScript [
   &EmitterState(Enable Distance Culling=true, MaxDistance=4500, Module.bUseMaxDistance=true, Module.bUseMinDistance=false, Scalability Mode=::NewEnumerator1)
   &SpawnBurst_Instantaneous(Use Spawn Probability=false, UseLoopCountLimit=false)
  ]
  particleSpawnScript [
   &InitializeParticle(
    Color=$System.HeartEmissive
    Lifetime=1.5
    Mesh Scale=<0.35 0.35 0.35>
    Mesh Scale Mode=::NewEnumerator4
    UsePositionOffset=false
    Write Material Random=true
   )
   &AddVelocity(Velocity=<0 120 0>)
   /Niagara/Modules/CascadeConversion/CascadeConversion_InitialMeshRotation(Rotation:Vector3f=&RandomRangeVector(Minimum=<-0.03 0 0>, Maximum=<0.03 0 0>))
   &InitialMeshOrientation(
    Mesh Orientation Mode=::NewEnumerator2
    Rotation=&RandomRangeVector(Minimum=<0 -0.03 0>, Maximum=<0 0.03 0>)
    Use Rotation Vector=true
   )
  ]
  particleUpdateScript [
   &ParticleState()
   &UpdateMeshOrientation(
    Rotation Rate=&FloatFromCurve(
     FloatCurve=DataInterface() {propertyValues {Curve {keys [
         {arriveTangent=-2.5448, interpMode=RCIM_Cubic, leaveTangent=-2.5448, tangentMode=RCTM_User, time=0, value=1}
         {arriveTangent=-0.7717, interpMode=RCIM_Cubic, leaveTangent=-0.7717, tangentMode=RCTM_User, time=0.4, value=0.22}
         <at 1 0>
        ]}}}
     Scale Curve=6
    )
    Rotation Vector=<0 0 1>
   )
   &ScaleMeshSize(Scale Factor=&VectorFromFloat(Value=&FloatFromCurve(FloatCurve=DataInterface() {propertyValues {Curve {keys [
          {arriveTangent=3.6183, interpMode=RCIM_Cubic, time=0, value=0.2}
          {arriveTangent=0.9033, interpMode=RCIM_Cubic, leaveTangent=0.9033, time=0.6831, value=1}
          {arriveTangent=-0.5286, interpMode=RCIM_Cubic, leaveTangent=-0.5286, time=0.8857, value=1}
          {arriveTangent=5.7236, interpMode=RCIM_Cubic, leaveTangent=5.7236, time=0.9498, value=0.859}
          <at 0.9988 1.6478 RCIM_Cubic>
         ]}}})))
   &Drag(Drag=2, Use Linear Drag=true, Use Rotational Drag=true)
   &AccelerationForce(Acceleration=<0 0 80>)
   &SolveForcesAndVelocity(Clamp Velocity=false, Limit Acceleration=false)
   &GenerateDeathEvent(Use Event Probability=false)
  ]
  rendererClasses [&NiagaraMeshRendererProperties]
  renderers [{rendererClass=&NiagaraMeshRendererProperties, rendererIndex=0}]
 }
 Empty {
  bEnabled=false
  emitterUpdateScript [
   &EmitterState()
   &SpawnBurst_Instantaneous(Use Spawn Probability=false, UseLoopCountLimit=false)
  ]
  particleSpawnScript [&InitializeParticle(Color Mode=::NewEnumerator0, Lifetime=5, Sprite Size Mode=::NewEnumerator3, UsePositionOffset=false, Write Material Random=true)]
  particleUpdateScript [&ParticleState()]
  rendererClasses [&NiagaraSpriteRendererProperties, &NiagaraMeshRendererProperties]
  renderers [
   {rendererClass=&NiagaraSpriteRendererProperties}
   {rendererClass=&NiagaraMeshRendererProperties, rendererIndex=1}
  ]
 }
 LittleHearts {
  emitterUpdateScript [
   &EmitterState(Enable Distance Culling=true, MaxDistance=3500, Module.bUseMaxDistance=true, Module.bUseMinDistance=false, Scalability Mode=::NewEnumerator1)
   &SpawnBurst_Instantaneous(Spawn Count=5, Use Spawn Probability=false, UseLoopCountLimit=false)
  ]
  particleSpawnScript [
   &InitializeParticle(
    Color=$System.HeartEmissive
    Lifetime=&RandomRangeFloat(Minimum=0.45, Maximum=1)
    Mesh Scale=&VectorFromFloat(Value=&RandomRangeFloat(Minimum=0.1, Maximum=0.16))
    Mesh Scale Mode=::NewEnumerator4
    UsePositionOffset=false
    Write Material Random=true
   )
   &AddVelocity(Velocity=&RandomRangeVector(Minimum=<-60 50 -40>, Maximum=<0 160 40>))
   &InitialMeshOrientation(
    Mesh Orientation Mode=::NewEnumerator2
    Rotation=&RandomRangeVector(Minimum=<0 -0.03 0>, Maximum=<0 0.03 0>)
    Use Rotation Vector=true
   )
  ]
  particleUpdateScript [
   &ParticleState()
   &ScaleMeshSize(Scale Factor=&VectorFromFloat(Value=&FloatFromCurve(FloatCurve=DataInterface() {propertyValues {Curve {keys [
          {arriveTangent=3.6183, interpMode=RCIM_Cubic, time=0, value=0.2}
          {arriveTangent=0.9936, interpMode=RCIM_Cubic, leaveTangent=0.9936, time=0.6831, value=1}
          {arriveTangent=1.4192, interpMode=RCIM_Cubic, leaveTangent=1.4192, time=0.8052, value=1}
          {arriveTangent=-5.1332, interpMode=RCIM_Cubic, leaveTangent=-5.1332, time=0.8911, value=1.2952}
          {interpMode=RCIM_Cubic, leaveTangent=-33.019, time=1, value=0}
         ]}}})))
   &Drag(Drag=2, Use Linear Drag=true, Use Rotational Drag=true)
   &AccelerationForce(Acceleration=<0 0 30>)
   &SolveForcesAndVelocity(Clamp Velocity=false, Limit Acceleration=false)
   &GenerateDeathEvent(Use Event Probability=false)
  ]
  rendererClasses [&NiagaraMeshRendererProperties]
  renderers [{rendererClass=&NiagaraMeshRendererProperties, rendererIndex=0}]
 }
 Pop {
  emitterUpdateScript [&EmitterState(Enable Distance Culling=true, MaxDistance=3000, Module.bUseMaxDistance=true, Module.bUseMinDistance=false, Scalability Mode=::NewEnumerator1)]
  particleSpawnScript [
   &InitializeParticle(
    Color=$System.HeartEmissive
    Lifetime=0.2
    Mesh Scale Max=<1 1 0.6>
    Mesh Scale Min=<0.5 0.5 0.45>
    Mesh Scale Mode=::NewEnumerator2
    UsePositionOffset=false
    Write Material Random=true
   )
   &InitialMeshOrientation(Mesh Axis To Orient=<0 0 -1>, Mesh Orientation Mode=::NewEnumerator2, Use Rotation Vector=false)
  ]
  particleUpdateScript [
   &ParticleState()
   &ScaleMeshSize(Scale Factor=&VectorFromFloat(Value=&FloatFromCurve(FloatCurve=DataInterface() {propertyValues {Curve {keys [<at 0 1>,<at 1 0>]}}})))
   &AccelerationForce(Acceleration=<0 0 -300>)
   &SolveForcesAndVelocity(Clamp Velocity=false, Limit Acceleration=false)
  ]
  rendererClasses [&NiagaraMeshRendererProperties]
  renderers [{rendererClass=&NiagaraMeshRendererProperties, rendererIndex=0}]
 }
 PopSmall {
  emitterUpdateScript [&EmitterState(Enable Distance Culling=true, MaxDistance=2000, Module.bUseMaxDistance=true, Module.bUseMinDistance=false, Scalability Mode=::NewEnumerator1)]
  particleSpawnScript [
   &InitializeParticle(
    Color=$System.HeartEmissive
    Lifetime=0.2
    Mesh Scale=<0.35 0.35 0.25>
    Mesh Scale Mode=::NewEnumerator4
    UsePositionOffset=false
    Write Material Random=true
   )
   &InitialMeshOrientation(Mesh Axis To Orient=<0 0 -1>, Mesh Orientation Mode=::NewEnumerator2, Use Rotation Vector=false)
  ]
  particleUpdateScript [
   &ParticleState()
   &ScaleMeshSize(Scale Factor=&VectorFromFloat(Value=&FloatFromCurve(FloatCurve=DataInterface() {propertyValues {Curve {keys [<at 0 1>,<at 1 0>]}}})))
   &AccelerationForce(Acceleration=<0 0 -300>)
   &SolveForcesAndVelocity(Clamp Velocity=false, Limit Acceleration=false)
  ]
  rendererClasses [&NiagaraMeshRendererProperties]
  renderers [{rendererClass=&NiagaraMeshRendererProperties, rendererIndex=0}]
 }
}
summary {
 systemName=NS_Emote_BlowKiss
 emitters [
  {emitterName=BigHeart, rendererClasses [&NiagaraMeshRendererProperties]}
  {emitterName=LittleHearts, rendererClasses [&NiagaraMeshRendererProperties]}
  {emitterName=Pop, rendererClasses [&NiagaraMeshRendererProperties]}
  {emitterName=PopSmall, rendererClasses [&NiagaraMeshRendererProperties]}
  {
   bEnabled=false
   emitterName=Empty
   rendererClasses [&NiagaraSpriteRendererProperties, &NiagaraMeshRendererProperties]
  }
 ]
 userVariables []
}
