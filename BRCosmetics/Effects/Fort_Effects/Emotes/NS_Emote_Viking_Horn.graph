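# Niagara effect for the Viking Horn emote, combining a mesh ring with a short-lived spray of small sprites.
# Ring1 bursts a mesh that expands and fades, with randomized dynamic material parameters.
# PlasmaDrips runs on GPU, spawning sprites on a sphere and launching them rapidly, with color flash, drag, curve-driven shrinking and speed-based stretch.
# Both emitters use distance culling; the Empty emitter is disabled.
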NiagaraSystem /BRCosmetics/Effects/Fort_Effects/Emotes/NS_Emote_Viking_Horn
dependencies {
 usedDataInterfaces [
  &NiagaraDataInterfaceVectorCurve
  &NiagaraDataInterfaceCurve
  &NiagaraDataInterfaceVector2DCurve
 ]
 usedDynamicInputs [&ConvertVectorToPosition, &RandomRangeVector]
 usedModules [&SystemState]
}
emitters {
 Empty {
  bEnabled=false
  emitterUpdateScript [
   &EmitterState()
   &SpawnBurst_Instantaneous(Use Spawn Probability=false, UseLoopCountLimit=false)
  ]
  particleSpawnScript [&InitializeParticle(Color Mode=::NewEnumerator0, Lifetime=5, UsePositionOffset=false, Write Material Random=true)]
  particleUpdateScript [&ParticleState()]
  rendererClasses [&NiagaraSpriteRendererProperties, &NiagaraMeshRendererProperties]
  renderers [
   {rendererClass=&NiagaraSpriteRendererProperties}
   {rendererClass=&NiagaraMeshRendererProperties, rendererIndex=1}
  ]
 }
 PlasmaDrips {
  simTarget=GPUComputeSim
  emitterUpdateScript [
   &EmitterState(
    Enable Distance Culling=true
    Life Cycle Mode=::NewEnumerator1
    Loop Behavior=::NewEnumerator1
    Loop Delay=0.2
    Loop Duration=2
    MaxDistance=1500
    Module.bUseMaxDistance=true
    Module.bUseMinDistance=false
    Scalability Mode=::NewEnumerator1
    UseLoopDelay=true
   )
   &SpawnRate(SpawnRate=150, Use Spawn Probability=false)
  ]
  particleSpawnScript [
   &InitializeParticle(
    Color=<rgba 0.6706 0.5412 0.4588 0.025>
    Lifetime=&RandomRangeFloat(Minimum=0.15, Maximum=0.25)
    Position Offset=<0 0 8>
    Sprite Size=&RandomRangeVector2D(Minimum=<2 1>, Maximum=<3 3>)
    Sprite Size Mode=::NewEnumerator4
    UsePositionOffset=true
    Write Material Random=true
   )
   &ShapeLocation(MaskSpawnGroup=false, Override Seed=false, Sphere Radius=4, Sphere Surface Distribution=1, UseOwnerScale=true)
   &AddVelocity(
    Module.Use Falloff Distance=false
    Origin Offset=<0 0 -18>
    Override Seed=false
    Velocity Mode=::NewEnumerator1
    Velocity Speed=&RandomRangeFloat(Minimum=500, Maximum=700)
   )
  ]
  particleUpdateScript [
   &ParticleState()
   &ScaleSpriteSize(
    Scale Factor=&Vector2DFromCurve(
     Vector2Curve=DataInterface() {propertyValues {
       XCurve {keys [
         {arriveTangent=3.6385, interpMode=RCIM_Cubic, time=0, value=0.5}
         {arriveTangent=-1.6029, leaveTangent=-1.6029, time=0.3, value=1}
         {arriveTangent=-1.2589, tangentMode=RCTM_User, time=1, value=0}
        ]}
       YCurve {keys [
         {arriveTangent=-16.1833, interpMode=RCIM_Cubic, time=0, value=3}
         {arriveTangent=-0.6675, leaveTangent=-0.6675, time=0.3, value=1}
         {arriveTangent=-0.8554, tangentMode=RCTM_User, time=1, value=0}
        ]}
      }}
     ScaleCurve=<1 1>
    )
    Scale Sprite Size Mode=::NewEnumerator2
   )
   &ScaleColor(
    Scale RGB=&VectorFromFloat(Value=&FloatFromCurve(FloatCurve=DataInterface() {propertyValues {Curve {keys [{arriveTangent=-36.7421, interpMode=RCIM_Cubic, time=0, value=8}, <at 0.25 1>]}}}))
    ScaleA=false
    ScaleRGB=true
   )
   &Drag(Use Linear Drag=true, Use Rotational Drag=false)
   &SolveForcesAndVelocity(Clamp Velocity=false, Limit Acceleration=false)
   &ScaleSpriteSizeBySpeed(Max Scale Factor=<1 2.5>, Source Velocity:Vector3f=&Multiply_Vector(A=<0 0.05 1>, B=$Particles.Velocity))
  ]
  rendererClasses [&NiagaraSpriteRendererProperties]
  renderers [{rendererClass=&NiagaraSpriteRendererProperties}]
 }
 Ring1 {
  emitterUpdateScript [
   &EmitterState(Enable Distance Culling=true, MaxDistance=2500, Module.bUseMaxDistance=true, Module.bUseMinDistance=false, Scalability Mode=::NewEnumerator1)
   &SpawnBurst_Instantaneous(Use Spawn Probability=false, UseLoopCountLimit=false)
  ]
  particleSpawnScript [
   &InitializeParticle(
    Color=<rgba 3 2.6 2.45 1>
    Lifetime=2
    Mesh Scale=<0.2 0.04 0.04>
    Mesh Scale Mode=::NewEnumerator4
    Position Offset=<0 0 5>
    UsePositionOffset=true
    Write Material Random=true
   )
   &DynamicMaterialParameters(
    Index 0 Param 1=&RandomRangeFloat(Minimum=0.5, Maximum=2)
    Index 0 Param 3=1
    Module.Index0Param0Write=true
    Module.Index0Param1Write=true
    Module.Index0Param2Write=true
    Module.Index0Param3Write=true
   )
   &InitialMeshOrientation(Mesh Orientation Mode=::NewEnumerator2, Orientation Vector=<0 0 1>, Use Rotation Vector=false)
  ]
  particleUpdateScript [
   &ParticleState()
   &ScaleMeshSize(Scale Factor=&VectorFromCurve(VectorCurve=DataInterface() {propertyValues {XCurve {keys [<at 0 0>,<at 0.25 1>]}, YCurve {keys [<at 0 1>,<at 0.25 1.25>]}, ZCurve {keys [<at 0 1>,<at 0.25 1>]}}}))
   &ScaleColor(
    Scale Alpha=&FloatFromCurve(FloatCurve=DataInterface() {propertyValues {Curve {keys [
         {arriveTangent=7.5084, interpMode=RCIM_Cubic, leaveTangent=7.5084, tangentMode=RCTM_User, time=0, value=0}
         {arriveTangentWeight=0.0923, interpMode=RCIM_Cubic, leaveTangentWeight=0.1475, tangentMode=RCTM_User, time=0.25, value=1}
         {arriveTangent=-1.3333, interpMode=RCIM_Cubic, leaveTangent=-1.3333, time=0.6496, value=0.3085}
         <at 1 0 RCIM_Cubic>
        ]}}})
    ScaleA=true
    ScaleRGB=true
   )
   &CameraOffset(Camera Offset Amount=-1)
  ]
  rendererClasses [&NiagaraMeshRendererProperties]
  renderers [{rendererClass=&NiagaraMeshRendererProperties, rendererIndex=0}]
 }
}
summary {
 systemName=NS_Emote_Viking_Horn
 emitters [
  {emitterName=Ring1, rendererClasses [&NiagaraMeshRendererProperties]}
  {emitterName=PlasmaDrips, simTarget=GPUComputeSim, rendererClasses [&NiagaraSpriteRendererProperties]}
  {
   bEnabled=false
   emitterName=Empty
   rendererClasses [&NiagaraSpriteRendererProperties, &NiagaraMeshRendererProperties]
  }
 ]
 userVariables []
}
